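# Physics asset for the Season 19 Umbrella glider mesh in the VinderTech Umbrella gadget cosmetics.
# It defines a single body on the Melee_TwoHanded_Handle bone, made of one offset, rotated box that serves as a coarse collision volume.
# It has no constraints, so the asset sets up no jointed or articulated simulation between bones.
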PhysicsAsset /BRCosmetics/Gadgets/Assets/VinderTech_Umbrella/Umbrella_Season19/Meshes/Season_19_Glider_Physics
bodies {Melee_TwoHanded_Handle [{shapeType=Box, extentX=153.9995, extentY=162.4813, extentZ=136.7596, center=<-2.3456 10.8041 62.8209>, rotation=<rot 0 103.1986 -81.9549>}]}
constraints []
